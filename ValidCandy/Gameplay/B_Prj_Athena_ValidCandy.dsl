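; Projectile-style Blueprint for the ValidCandy item that spawns a BGA (building gameplay actor) through its parent class, likely on impact.
; The SpawnBGA override calls the parent with the hit data, casts the spawned actor to BGA_Athena_ValidCandyCore, copies its own AbilityLevel onto it and returns it.
; A failed cast only prints a debug string.
; Event stubs are empty and the construction script just calls the parent.
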(event EventBeginPlay)

(event Collision|EventActorBeginOverlap
  (params (OtherActor :type Actor%20Object%20Reference)))

(event EventTick
  (params (DeltaSeconds :type Float%20%28single-precision%29)))

(event EventDestroyed)

(fn SpawnBGA_
  (params (Normal :type Vector :default "0, 0, 0") (HitActor :type Actor%20Object%20Reference) (HitComp :type Primitive%20Component%20Object%20Reference) (Pos :type Vector :default "0, 0, 0") (FromWater :type Boolean) (Hit :type Hit%20Result%20Structure))
  (bind _returnvalue (|CallParentFunction Normal HitActor HitComp Pos FromWater Hit))
  (bind _asbga_athena_valid_candy_core (Utilities|Casting|CastToBGA_Athena_ValidCandyCore _returnvalue)
    (:then
      (Variables|/ValidCandy/Gameplay/BGA_Athena_ValidCandyCore.BGA_Athena_ValidCandyCore_C|SetAbilityLevel (Variables|Default|GetAbilityLevel) _asbga_athena_valid_candy_core)
      (return _returnvalue))
    (:CastFailed
      (Development|PrintString "CAST FAILED"))))

(fn ConstructionScript
  (|CallParentFunction))
